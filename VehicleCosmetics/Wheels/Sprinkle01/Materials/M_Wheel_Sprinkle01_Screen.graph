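# Unlit masked material for the Sprinkle01 wheel screen, playing a looping animated title card over a blue gradient.
# One offset Time loop zooms cloud-atlas layers and a two-frame title, switching composite order so the title moves from behind the clouds to the front.
# Sine flicker and vertex color modulate brightness, with stepped anime-light shading on High and Epic.
# Screen Mask drives opacity, and MF_Vehicle_FNPostFX finishes the output.
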Material /VehicleCosmetics/Wheels/Sprinkle01/Materials/M_Wheel_Sprinkle01_Screen
graph {
 properties {
  bAutomaticallySetUsageInEditor=false
  bUseMaterialAttributes=true
  bUsedWithMeshDeformer=true
  bUsedWithSkeletalMesh=true
  blendMode=BLEND_Masked
  shadingModel=MSM_Unlit
  usedShadingModels=Unlit
 }
 nodes [
  n1 = LinearInterpolate(
   A=Constant3Vector() {constant=<rgba 0.002 0.173 0.56 0>}
   B=Constant3Vector() {constant=<rgba 0.114 0.421 0.612 0>}
   Alpha=ComponentMask(None=TextureCoordinate()) {g=true}
  )
  n2 = VectorParameter() {parameterName=Title Dropshadow Color, defaultValue=<rgba 0.07 0.08 0.1 1>}
  n3 = TextureCoordinate()
  n4 = Constant2Vector() {g=-0.45, r=-0.4}
  n5 = Time()
  n6 = Frac(None=Add(A=Multiply(A=@n5) {constB=0.15}, B=ScalarParameter() {defaultValue=0.72, parameterName=Thumbnail Timing Offset}))
  n7 = LinearInterpolate(
   A=Multiply(A=Add(A=@n3, B=@n4)) {constB=5}
   B=Add(A=Multiply(A=@n3) {constB=0}, B=Constant2Vector() {g=0.32, r=0.43})
   Alpha=ConstantBiasScale(None=@n6) {bias=0, scale=1}
  )
  n8 = Constant2Vector() {g=0.5}
  n9 = Frac(None=Multiply(A=Time(), B=Constant() {r=6}))
  n10 = Constant() {r=0.5}
  n11 = Constant()
  n12 = Constant() {r=1}
  n13 = TextureSample(UVs=LinearInterpolate(A=@n7, B=Add(A=@n7, B=@n8), Alpha=If(A=@n9, B=@n10, "A > B"=@n11, "A == B"=@n11, "A < B"=@n12))) {samplerType=SAMPLERTYPE_LinearColor, texture=/VehicleCosmetics/Wheels/Sprinkle01/Textures/T_Wheel_Sprinkle01_Title_RGB}
  n14 = ComponentMask(None=@n7) {g=true}
  n15 = Constant() {r=0.01}
  n16 = Constant() {r=1}
  n17 = Constant()
  n18 = Constant() {r=0.49}
  n19 = Constant() {r=1}
  n20 = Constant()
  n21 = Constant() {r=0.01}
  n22 = Constant() {r=1}
  n23 = Constant()
  n24 = ComponentMask(None=@n7) {r=true}
  n25 = Constant() {r=0.99}
  n26 = Constant()
  n27 = Constant() {r=1}
  n28 = Multiply(
   A=Multiply(A=If(A=@n14, B=@n15, "A > B"=@n16, "A == B"=@n17, "A < B"=@n17), B=If(A=@n14, B=@n18, "A > B"=@n20, "A == B"=@n20, "A < B"=@n19))
   B=Multiply(A=If(A=@n24, B=@n21, "A > B"=@n22, "A == B"=@n23, "A < B"=@n23), B=If(A=@n24, B=@n25, "A > B"=@n26, "A == B"=@n26, "A < B"=@n27))
  )
  n29 = Multiply(A=@n28, B=@n13.G)
  n30 = VectorParameter() {parameterName=Title Text Color, defaultValue=<rgba 0.8 0.4 0 1>}
  n31 = Multiply(A=@n28, B=@n13.R)
  n32 = LinearInterpolate(A=LinearInterpolate(A=@n1, B=@n2.RGB, Alpha=@n29), B=@n30.RGB, Alpha=@n31)
  n33 = TextureCoordinate()
  n34 = LinearInterpolate(
   A=Add(A=Multiply(A=@n33), B=Constant2Vector())
   B=Add(A=Multiply(A=@n33) {constB=0.7}, B=Constant2Vector() {g=0.15, r=0.2})
   Alpha=ConstantBiasScale(None=@n6) {bias=0, scale=1}
  )
  n35 = Constant() {r=1}
  n36 = ComponentMask(None=@n34) {g=true}
  n37 = Constant()
  n38 = If(A=@n36, B=Constant(), "A > B"=@n35, "A == B"=@n37, "A < B"=@n37)
  n39 = If(A=@n36, B=Constant() {r=0.2}, "A > B"=@n37, "A == B"=@n37, "A < B"=@n35)
  n40 = TextureSample(UVs=@n34) {texture=/VehicleCosmetics/Wheels/Sprinkle01/Textures/T_Wheel_Sprinkle01_CloudAtlas_BC}
  n41 = Multiply(A=If(A=ComponentMask(None=@n34) {r=true}, B=Constant() {r=0.6}, "A > B"=@n37, "A == B"=@n37, "A < B"=Multiply(A=@n38, B=@n39)), B=@n40.A)
  n42 = TextureCoordinate()
  n43 = AppendVector(A=OneMinus(None=ComponentMask(None=@n42) {r=true}), B=ComponentMask(None=@n42) {g=true})
  n44 = LinearInterpolate(
   A=Add(A=Multiply(A=@n43) {constB=1.3}, B=Constant2Vector() {g=0.35, r=-0.24})
   B=Add(A=Multiply(A=@n43) {constB=1.2}, B=Constant2Vector() {g=0.43, r=-0.3})
   Alpha=ConstantBiasScale(None=@n6) {bias=0, scale=1}
  )
  n45 = TextureSample(UVs=@n44) {texture=&T_Wheel_Sprinkle01_CloudAtlas_BC}
  n46 = ComponentMask(None=@n44) {g=true}
  n47 = Constant() {r=1}
  n48 = Constant()
  n49 = If(A=@n46, B=Constant() {r=0.6}, "A > B"=@n47, "A == B"=@n48, "A < B"=@n48)
  n50 = If(A=@n46, B=Constant() {r=0.7}, "A > B"=@n48, "A == B"=@n48, "A < B"=@n47)
  n51 = Multiply(A=If(A=ComponentMask(None=@n44) {r=true}, B=Constant() {r=0.3}, "A > B"=Multiply(A=@n49, B=@n50), "A == B"=@n48, "A < B"=@n48), B=@n45.A)
  n52 = TextureCoordinate()
  n53 = LinearInterpolate(
   A=Add(A=Multiply(A=@n52) {constB=0.75}, B=Constant2Vector() {g=0.3, r=0.05})
   B=Add(A=Multiply(A=@n52) {constB=0.6}, B=Constant2Vector() {g=0.375, r=0.2})
   Alpha=ConstantBiasScale(None=@n6) {bias=0, scale=1}
  )
  n54 = TextureSample(UVs=@n53) {texture=&T_Wheel_Sprinkle01_CloudAtlas_BC}
  n55 = Constant() {r=1}
  n56 = Constant()
  n57 = ComponentMask(None=@n53) {g=true}
  n58 = If(A=@n57, B=Constant() {r=0.7}, "A > B"=@n55, "A == B"=@n56, "A < B"=@n56)
  n59 = If(A=@n57, B=Constant() {r=0.8}, "A > B"=@n56, "A == B"=@n56, "A < B"=@n55)
  n60 = Multiply(A=If(A=ComponentMask(None=@n53) {r=true}, B=Constant() {r=0.5}, "A > B"=@n56, "A == B"=@n56, "A < B"=Multiply(A=@n58, B=@n59)), B=@n54.A)
  n61 = LinearInterpolate(A=LinearInterpolate(A=LinearInterpolate(A=@n32, B=@n40.RGB, Alpha=@n41), B=@n45.RGB, Alpha=@n51), B=@n54.RGB, Alpha=@n60)
  n62 = TextureCoordinate()
  n63 = LinearInterpolate(
   A=Add(A=Multiply(A=@n62) {constB=0.6}, B=Constant2Vector() {r=0.15})
   B=Add(A=Multiply(A=@n62) {constB=0.4}, B=Constant2Vector() {g=0.12, r=0.35})
   Alpha=ConstantBiasScale(None=@n6) {bias=0, scale=1}
  )
  n64 = TextureSample(UVs=@n63) {texture=&T_Wheel_Sprinkle01_CloudAtlas_BC}
  n65 = Constant() {r=1}
  n66 = Constant()
  n67 = ComponentMask(None=@n63) {g=true}
  n68 = If(A=@n67, B=Constant() {r=0.2}, "A > B"=@n65, "A == B"=@n66, "A < B"=@n66)
  n69 = If(A=@n67, B=Constant() {r=0.4}, "A > B"=@n66, "A == B"=@n66, "A < B"=@n65)
  n70 = Multiply(A=If(A=ComponentMask(None=@n63) {r=true}, B=Constant() {r=0.5}, "A > B"=@n66, "A == B"=@n66, "A < B"=Multiply(A=@n68, B=@n69)), B=@n64.A)
  n71 = TextureCoordinate()
  n72 = LinearInterpolate(
   A=Add(A=Multiply(A=@n71) {constB=0.8}, B=Constant2Vector() {g=0.05, r=0.16})
   B=Add(A=Multiply(A=@n71) {constB=0.6}, B=Constant2Vector() {g=0.15, r=0.25})
   Alpha=ConstantBiasScale(None=@n6) {bias=0, scale=1}
  )
  n73 = TextureSample(UVs=@n72) {texture=&T_Wheel_Sprinkle01_CloudAtlas_BC}
  n74 = Constant() {r=1}
  n75 = Constant()
  n76 = ComponentMask(None=@n72) {g=true}
  n77 = If(A=@n76, B=Constant(), "A > B"=@n74, "A == B"=@n75, "A < B"=@n75)
  n78 = If(A=@n76, B=Constant() {r=0.2}, "A > B"=@n75, "A == B"=@n75, "A < B"=@n74)
  n79 = Multiply(A=If(A=ComponentMask(None=@n72) {r=true}, B=Constant() {r=0.6}, "A > B"=Multiply(A=@n77, B=@n78), "A == B"=@n75, "A < B"=@n75), B=@n73.A)
  n80 = TextureCoordinate()
  n81 = LinearInterpolate(
   A=Add(A=Multiply(A=@n80) {constB=0.75}, B=Constant2Vector() {g=0.1, r=0.23})
   B=Add(A=Multiply(A=@n80) {constB=0.6}, B=Constant2Vector() {g=0.175, r=0.2})
   Alpha=ConstantBiasScale(None=@n6) {bias=0, scale=1}
  )
  n82 = TextureSample(UVs=@n81) {texture=&T_Wheel_Sprinkle01_CloudAtlas_BC}
  n83 = Constant() {r=1}
  n84 = Constant()
  n85 = ComponentMask(None=@n81) {g=true}
  n86 = If(A=@n85, B=Constant() {r=0.2}, "A > B"=@n83, "A == B"=@n84, "A < B"=@n84)
  n87 = If(A=@n85, B=Constant() {r=0.4}, "A > B"=@n84, "A == B"=@n84, "A < B"=@n83)
  n88 = Multiply(A=If(A=ComponentMask(None=@n81) {r=true}, B=Constant() {r=0.5}, "A > B"=Multiply(A=@n86, B=@n87), "A == B"=@n84, "A < B"=@n84), B=@n82.A)
  n89 = LinearInterpolate(A=LinearInterpolate(A=LinearInterpolate(A=@n61, B=@n64.RGB, Alpha=@n70), B=@n73.RGB, Alpha=@n79), B=@n82.RGB, Alpha=@n88)
  n90 = TextureCoordinate()
  n91 = LinearInterpolate(
   A=Add(A=Multiply(A=@n90) {constB=0.8}, B=Constant2Vector() {g=0.125, r=0.18})
   B=Add(A=Multiply(A=@n90) {constB=0.4}, B=Constant2Vector() {g=0.35, r=0.165})
   Alpha=ConstantBiasScale(None=@n6) {bias=0, scale=1}
  )
  n92 = Constant() {r=1}
  n93 = Constant()
  n94 = ComponentMask(None=@n91) {g=true}
  n95 = If(A=@n94, B=Constant() {r=0.4}, "A > B"=@n92, "A == B"=@n93, "A < B"=@n93)
  n96 = If(A=@n94, B=Constant() {r=0.6}, "A > B"=@n93, "A == B"=@n93, "A < B"=@n92)
  n97 = TextureSample(UVs=@n91) {texture=&T_Wheel_Sprinkle01_CloudAtlas_BC}
  n98 = Multiply(A=If(A=ComponentMask(None=@n91) {r=true}, B=Constant() {r=0.4}, "A > B"=Multiply(A=@n95, B=@n96), "A == B"=@n93, "A < B"=@n93), B=@n97.A)
  n99 = TextureCoordinate()
  n100 = LinearInterpolate(
   A=Add(A=Multiply(A=@n99) {constB=0.95}, B=Constant2Vector() {g=0.06})
   B=Add(A=Multiply(A=@n99) {constB=0.8}, B=Constant2Vector() {g=0.1})
   Alpha=ConstantBiasScale(None=@n6) {bias=0, scale=1}
  )
  n101 = Constant() {r=1}
  n102 = Constant()
  n103 = ComponentMask(None=@n100) {g=true}
  n104 = If(A=@n103, B=Constant() {r=0.6}, "A > B"=@n101, "A == B"=@n102, "A < B"=@n102)
  n105 = If(A=@n103, B=Constant() {r=0.7}, "A > B"=@n102, "A == B"=@n102, "A < B"=@n101)
  n106 = TextureSample(UVs=@n100) {texture=&T_Wheel_Sprinkle01_CloudAtlas_BC}
  n107 = Multiply(
   A=If(A=ComponentMask(None=@n100) {r=true}, B=Constant() {r=0.3}, "A > B"=Multiply(A=@n104, B=@n105), "A == B"=@n102, "A < B"=@n102)
   B=@n106.A
  )
  n108 = TextureCoordinate()
  n109 = LinearInterpolate(
   A=Add(A=Multiply(A=@n108) {constB=0.7}, B=Constant2Vector() {g=0.3, r=0.2})
   B=Add(A=Multiply(A=@n108) {constB=0.45}, B=Constant2Vector() {g=0.4, r=0.3})
   Alpha=ConstantBiasScale(None=@n6) {bias=0, scale=1}
  )
  n110 = Constant() {r=1}
  n111 = Constant()
  n112 = ComponentMask(None=@n109) {g=true}
  n113 = If(A=@n112, B=Constant() {r=0.7}, "A > B"=@n110, "A == B"=@n111, "A < B"=@n111)
  n114 = If(A=@n112, B=Constant() {r=0.9}, "A > B"=@n111, "A == B"=@n111, "A < B"=@n110)
  n115 = TextureSample(UVs=@n109) {texture=&T_Wheel_Sprinkle01_CloudAtlas_BC}
  n116 = Multiply(
   A=If(A=ComponentMask(None=@n109) {r=true}, B=Constant() {r=0.5}, "A > B"=Multiply(A=@n113, B=@n114), "A == B"=@n111, "A < B"=@n111)
   B=@n115.A
  )
  n117 = LinearInterpolate(A=LinearInterpolate(A=LinearInterpolate(A=@n89, B=@n97.RGB, Alpha=@n98), B=@n106.RGB, Alpha=@n107), B=@n115.RGB, Alpha=@n116)
  n118 = TextureCoordinate()
  n119 = LinearInterpolate(
   A=Add(A=Multiply(A=@n118) {constB=0.8}, B=Constant2Vector() {g=0.3, r=0.05})
   B=Add(A=Multiply(A=@n118) {constB=0.4}, B=Constant2Vector() {g=0.52, r=0.25})
   Alpha=ConstantBiasScale(None=@n6) {bias=0, scale=1}
  )
  n120 = TextureSample(UVs=@n119) {texture=&T_Wheel_Sprinkle01_CloudAtlas_BC}
  n121 = Constant() {r=0.5}
  n122 = Constant() {r=1}
  n123 = Constant()
  n124 = ComponentMask(None=@n119) {g=true}
  n125 = If(A=@n124, B=Constant() {r=1}, "A > B"=@n123, "A == B"=@n123, "A < B"=@n122)
  n126 = Multiply(A=If(A=@n124, B=Constant() {r=0.8}, "A > B"=@n122, "A == B"=@n123, "A < B"=@n123), B=@n125)
  n127 = ComponentMask(None=@n119) {r=true}
  n128 = Constant() {r=0.7}
  n129 = Multiply(A=@n125, B=If(A=@n124, B=Constant() {r=0.9}, "A > B"=@n122, "A == B"=@n123, "A < B"=@n123))
  n130 = Multiply(
   A=@n120.A
   B=Max(A=If(A=@n127, B=@n121, "A > B"=@n123, "A == B"=@n123, "A < B"=@n126), B=If(A=@n127, B=@n128, "A > B"=@n123, "A == B"=@n123, "A < B"=@n129))
  )
  n131 = LinearInterpolate(A=@n1, B=@n40.RGB, Alpha=@n41)
  n132 = LinearInterpolate(A=LinearInterpolate(A=LinearInterpolate(A=@n131, B=@n45.RGB, Alpha=@n51), B=@n54.RGB, Alpha=@n60), B=@n64.RGB, Alpha=@n70)
  n133 = LinearInterpolate(A=@n132, B=@n2.RGB, Alpha=@n29)
  n134 = LinearInterpolate(A=LinearInterpolate(A=LinearInterpolate(A=@n133, B=@n30.RGB, Alpha=@n31), B=@n73.RGB, Alpha=@n79), B=@n82.RGB, Alpha=@n88)
  n135 = LinearInterpolate(A=LinearInterpolate(A=LinearInterpolate(A=@n134, B=@n97.RGB, Alpha=@n98), B=@n106.RGB, Alpha=@n107), B=@n115.RGB, Alpha=@n116)
  n136 = ConstantBiasScale(None=@n6) {bias=-0.2, scale=10}
  n137 = LinearInterpolate(A=LinearInterpolate(A=LinearInterpolate(A=@n132, B=@n73.RGB, Alpha=@n79), B=@n82.RGB, Alpha=@n88), B=@n97.RGB, Alpha=@n98)
  n138 = LinearInterpolate(A=LinearInterpolate(A=LinearInterpolate(A=@n137, B=@n106.RGB, Alpha=@n107), B=@n115.RGB, Alpha=@n116), B=@n120.RGB, Alpha=@n130)
  n139 = LinearInterpolate(
   A=LinearInterpolate(A=LinearInterpolate(A=@n117, B=@n120.RGB, Alpha=@n130), B=LinearInterpolate(A=@n135, B=@n120.RGB, Alpha=@n130), Alpha=Clamp(None=@n136))
   B=LinearInterpolate(A=LinearInterpolate(A=@n138, B=@n2.RGB, Alpha=@n29), B=@n30.RGB, Alpha=@n31)
   Alpha=Clamp(None=ConstantBiasScale(None=@n6) {bias=-0.4, scale=10})
  )
  n140 = Multiply(A=@n28, B=@n13.B)
  n141 = OneMinus(None=@n29)
  n142 = LinearInterpolate(
   A=@n139
   B=LinearInterpolate(A=@n139, B=@n1, Alpha=Multiply(A=@n140, B=@n141))
   Alpha=Clamp(None=ConstantBiasScale(None=@n6) {bias=-0.75, scale=12})
  )
  n143 = Clamp(None=ConstantBiasScale(None=@n6) {bias=-0.075, scale=5})
  n144 = Time()
  n145 = ConstantBiasScale(None=Sine(None=Multiply(A=@n144) {constB=10}))
  n146 = VertexColor()
  n147 = Multiply(
   A=Multiply(A=LinearInterpolate(A=@n1, B=@n142, Alpha=@n143), B=Clamp(None=@n145) {minDefault=0.91})
   B=Clamp(None=ConstantBiasScale(None=@n146.R) {bias=0.25, scale=1})
  )
  n148 = Multiply(A=@n147) {constB=2}
  n149 = DotProduct(
   A=MaterialFunctionCall() {materialFunction=/Game/Global/Materials/Character_Parent_2023/Anime_Parent_2023/MaterialFunctions/MFC_AnimeShader_LightDirection}.Direction Vector
   B=VertexNormalWS()
  )
  n150 = LinearInterpolate(A=@n148, B=Multiply(A=@n147) {constB=1.5}, Alpha=Saturate(None=ConstantBiasScale(None=@n149) {bias=-0.5, scale=10}))
  n151 = QualitySwitch(Default=@n148, High=@n150, Epic=@n150)
  n152 = TextureSampleParameter2D() {parameterName=Screen Mask, samplerType=SAMPLERTYPE_LinearColor, texture=/VehicleCosmetics/Wheels/Sprinkle01/Textures/T_Rim_Sprinkle01_RGB}
  MaterialFunctionCall!(In=SetMaterialAttributes!(MaterialAttributes, Emissive Color=@n151, Opacity Mask=@n152.R, Base Color=@n151) {attributeSetTypes [#1, #2, #3]}) {materialFunction=/VehicleCosmetics/Materials/MaterialFunctions/MF_Vehicle_FNPostFX}
 ]
}
